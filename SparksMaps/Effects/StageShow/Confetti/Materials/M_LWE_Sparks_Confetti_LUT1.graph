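# Unlit, masked, two-sided confetti material for Sparks stage show Niagara sprite and mesh particles.
# A curve-atlas palette, picked by particle random or vertex color, blends with stage colors and drives sheen and flipbook-normal shading in emissive.
# A second flipbook supplies alpha.
# The IsMeshCanBurst switch enables cell-noise glints and WPO for per-piece sprite sizing, spread, fall and wobble, with optional fake motion blur via PreviousFrameSwitch, bypassed on mobile.
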Material /SparksMaps/Effects/StageShow/Confetti/Materials/M_LWE_Sparks_Confetti_LUT1
graph {
 properties {
  bEnableResponsiveAA=true
  bHasPixelAnimation=true
  bUsedWithNiagaraMeshParticles=true
  bUsedWithNiagaraSprites=true
  blendMode=BLEND_Masked
  opacityMaskClipValue=0.1
  pixelDepthOffsetMode=PDOM_Legacy
  shadingModel=MSM_Unlit
  twoSided=true
  usedShadingModels=Unlit
 }
 nodes [
  n1 = NamedRerouteUsage()
  n2 = Add(A=Multiply(A=TextureCoordinate()), B=Add(A=Multiply(A=@n1) {constB=11.77}, B=Time()))
  n3 = Sine(None=Multiply(A=ComponentMask(None=@n2) {g=true}) {constB=0.25})
  ConstantBiasScale(None=@n3)
  n4 = TextureSample(UVs=Panner(Coordinate=TextureCoordinate() {coordinateIndex=1, uTiling=4, vTiling=4}) {speedX=0.5, speedY=0.25}) {samplerType=SAMPLERTYPE_LinearGrayscale, texture=/Game/Effects/Fort_Effects/Textures/Tile/T_CellNoise_Tile_M}
  n5 = Constant() {r=4}
  n6 = ScalarParameter() {defaultValue=0.01, group=1. Confetti Surface, parameterName=Sheen_Min}
  n7 = ScalarParameter() {defaultValue=0.5, group=1. Confetti Surface, parameterName=Sheen_Max}
  n8 = Saturate(None=@n3)
  n9 = MaterialFunctionCall!(CustomUPosition) {materialFunction=/SparksStageShow/Systems/StageManager/FXManager/MaterialFunctions/MF_MojoStageColors}
  n10 = Saturate(None=Add(A=@n9.Primary) {constB=0.1})
  n11 = ScalarParameter() {defaultValue=0.5, group=1. Confetti Surface, parameterName=StageShow Contribution, sliderMax=1}
  n12 = DynamicParameter() {parameterIndex=3, defaultValue=<rgba 0 0 0 0>, paramNames [Particle Random 0-1, "", "", ""]}
  n13 = CurveAtlasRowParameter(CurveTime=StaticSwitchParameter(
    True=VertexColor().G
    False=NamedRerouteDeclaration(None=@n12.Particle Random 0-1) {name=ParticleRandom 0-1 temp, nodeColor=<rgba 1 0.8906 0 1>}
   ) {parameterName=UseVertColorRandomColor}) {
   atlas=/SparksMaps/Effects/StageShow/Confetti/Textures/CA_Sparks_Confetti_Palettes
   curve=/SparksMaps/Effects/StageShow/Confetti/Textures/C_Sparks_Confetti_Rainbow
   group=1. Confetti Surface
   parameterName=Curve_ColorPalette
  }
  n14 = Multiply(A=@n13)
  n15 = Multiply(
   A=StaticSwitchParameter(True=Power(Base=@n4.R, Exp=@n5), False=LinearInterpolate(A=@n6, B=@n7, Alpha=@n8)) {parameterName=IsMeshCanBurst}
   B=Multiply(A=@n14, B=LinearInterpolate(B=@n10, Alpha=@n11) {constA=1})
  )
  n16 = Constant() {r=5}
  n17 = Add(A=@n9.Primary) {constB=0.5}
  n18 = NamedRerouteUsage()
  n19 = TextureObjectParameter() {parameterName=Confetti_Normal0, samplerType=SAMPLERTYPE_Normal, texture=/Game/Effects/Fort_Effects/Textures/Confetti/T_Confetti_SubUV_4X4_03}
  n20 = MaterialFunctionCall("Animation  Phase (0-1)"=NamedRerouteUsage(), Number of Rows=@n18, Number of Columns=@n18, Texture=@n19) {materialFunction=&FlipBook}
  n21 = Constant3Vector() {constant=<rgba 0 0 1 0>}
  n22 = LinearInterpolate(A=TextureCoordinate(), B=ComponentMask(None=@n20.Result) {g=true, r=true}) {constAlpha=0.2}
  n23 = Time()
  n24 = NamedRerouteUsage()
  n25 = Multiply(A=Add(A=ComponentMask(None=@n22) {g=true}, B=Add(A=@n23, B=@n24))) {constB=1.5}
  n26 = Add(A=Power(Base=DotProduct(A=@n20.Result, B=@n21)), B=ConstantBiasScale(None=Sine(None=@n25)))
  n27 = Multiply(
   A=LinearInterpolate(B=Saturate(None=@n17), Alpha=@n11) {constA=1}
   B=Multiply(
    A=Multiply(A=@n14, B=@n26)
    B=MaterialFunctionCall() {materialFunction=/Game/Effects/Fort_Effects/Materials/_Functions/CurveColor_Time_of_Day_S23}.Result
   )
  )
  n28 = Constant() {desc="Shadowed (non light transmission) intensity", r=0.75}
  MaterialFunctionCall(Emissive=Add(A=Multiply(A=@n15, B=@n16) {constB=2}, B=Multiply(A=@n27, B=@n28) {constB=0.15})) {materialFunction=&MF_FX_EmissiveAdjustments}
  n29 = Time()
  n30 = NamedRerouteUsage()
  NamedRerouteDeclaration(None=Frac(None=Add(A=@n29, B=@n30))) {name=Animation Phase, nodeColor=<rgba 0 1 0.2031 1>}
  NamedRerouteDeclaration(None=Constant() {r=6}) {name=RowsAndColumns, nodeColor=<rgba 0 1 0.4375 1>}
  n31 = Multiply(A=VertexColor().G) {constB=25}
  NamedRerouteDeclaration(None=StaticSwitchParameter(True=Divide(A=@n31) {constB=50.33}, False=NamedRerouteUsage()) {parameterName=UseVertColorRandomFlipbookStart}) {name=Random, nodeColor=<rgba 0 0.7422 1 1>}
  n32 = Constant()
  n33 = AppendVector(A=@n32, B=@n32)
  n34 = DynamicParameter() {defaultValue=<rgba 10 0 0 0>, paramNames [QuadScale,Animation,SpriteFall,WobbleAmount]}
  n35 = VertexColor()
  NamedRerouteDeclaration(None=Multiply(A=AppendVector(A=@n33, B=@n34.SpriteFall), B=Clamp(None=@n35) {minDefault=0.5})) {name=Downwards Movement, nodeColor=<rgba 0.5703 0 1 1>}
  n36 = NamedRerouteUsage()
  n37 = MaterialFunctionCall(
   "Animation  Phase (0-1)"=NamedRerouteUsage()
   Number of Rows=@n36
   Number of Columns=@n36
   Texture=TextureObject() {texture=/Game/Effects/Fort_Effects/Textures/Confetti/T_Confetti_SubUV_4X4_A}
  ) {materialFunction=&FlipBook}
  NamedRerouteDeclaration(None=Multiply(A=ParticleColor().A, B=ComponentMask(None=@n37.Result) {b=true})) {name=Alpha, nodeColor=<rgba 0.4062 0 1 1>}
  n38 = VertexColor()
  n39 = DynamicParameter() {defaultValue=<rgba 10 0 0 0>, paramNames [QuadScale,Animation,SpriteFall,WobbleAmount]}
  n40 = MaterialFunctionCall!(
   Center Location=WorldPosition().XYZ
   0-1 UVs
   XY scale=Multiply(A=Clamp(None=@n38.R) {minDefault=0.5}, B=@n39.QuadScale)
   OverridePivot
   "Normals (Optional)"
  ) {materialFunction=&Sprite}
  n41 = Constant3Vector()
  n42 = Multiply(A=Subtract(A=WorldPosition().XYZ, B=TransformPosition(None=@n41) {transformType=TRANSFORMPOSSOURCE_World}), B=@n39.Animation)
  n43 = Clamp(None=VertexColor()) {minDefault=0.25}
  n44 = Add(
   A=Add(A=NamedRerouteDeclaration(None=@n40.World position offset) {name=MeshSpriteSize, nodeColor=<rgba 0 0.6016 1 1>}, B=Multiply(A=@n42, B=@n43))
   B=NamedRerouteUsage()
  )
  n45 = VertexColor()
  n46 = Multiply(A=Time(), B=LinearInterpolate(Alpha=@n45.R) {constA=-1})
  n47 = Sine(None=Add(A=@n46, B=@n45.R))
  n48 = Constant() {r=25}
  n49 = Cosine(None=Add(A=@n46, B=@n45.B))
  n50 = Multiply(A=@n39.WobbleAmount, B=MaterialXAppend3Vector!(A=Multiply(A=@n47, B=@n48), B=Multiply(A=@n49, B=@n48), C=Constant()))
  NamedRerouteDeclaration(None=StaticSwitchParameter(True=Add(A=@n44, B=@n50), False=Constant()) {parameterName=IsMeshCanBurst}) {name=WPO, nodeColor=<rgba 0 1 0.2969 1>}
  NamedRerouteUsage()
  NamedRerouteUsage()
  ParticleRandom()
  ParticleRandom()
  ParticleRandom()
  ScalarParameter() {defaultValue=0.5, group=1. Confetti Surface, parameterName=TestFloat}
  n51 = Constant()
  n52 = NamedRerouteUsage()
  n53 = PreviousFrameSwitch(Current Frame=@n52, Previous Frame=AppendVector(A=AppendVector(A=@n51, B=@n51), B=Constant() {r=-100}))
  n54 = DynamicParameter() {parameterIndex=1, defaultValue=<rgba 0 0 0 0>, paramNames [MotionBlurDirectionalAmountWPO, Param2, Param3, Param4]}
  ShadingPathSwitch(
   Default=StaticSwitchParameter(True=LinearInterpolate(A=@n52, B=@n53, Alpha=@n54.MotionBlurDirectionalAmountWPO), False=@n52) {defaultValue=true, parameterName=UseFakeMotionBlur}
   Mobile=@n52
  )
  VectorParameter() {group=1. Confetti Surface, parameterName=TestColor, defaultValue=<rgba 1 0 0 1>}
 ]
}
